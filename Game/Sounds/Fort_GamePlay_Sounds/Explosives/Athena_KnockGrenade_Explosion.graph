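# Sound cue for the Athena knock grenade explosion, layering a close blast with a medium-distance one.
# The close layer randomly picks one of three Outlander shaped-charge waves; the medium layer picks one of two distant explosion waves, pitched up.
# Each layer has its own Athena projectile attenuation asset before the mixer combines them.
# The cue uses the large explosives sound class and overrides concurrency to prevent new instances at the limit.
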SoundCue /Game/Sounds/Fort_GamePlay_Sounds/Explosives/Athena_KnockGrenade_Explosion
root=Mixer_0
nodes [
 {
  id=Attenuation_0
  children [Random_0]
  properties {attenuationSettings=&Fort_Weapon_Projectile_Close_Athena_Att, attenuationOverrides {bAttenuate=false}}
 }
 {
  id=Attenuation_4
  children [Modulator_2]
  properties {attenuationSettings=&Fort_Weapon_Projectile_Medium_Mono_Athena_Att, attenuationOverrides {bAttenuate=false}}
 }
 {id=Mixer_0, children [Attenuation_0, Attenuation_4], properties {inputVolume [1,1]}}
 {id=Modulator_2, children [Random_1], properties {pitchMax=1.3, pitchMin=1.3}}
 {id=Random_0, children [WavePlayer_3,WavePlayer_4,WavePlayer_6], properties {weights [1,1,1]}}
 {id=Random_1, children [WavePlayer_1,WavePlayer_7], properties {weights [1,1]}}
 {id=WavePlayer_1, properties {soundWaveAssetPtr=&/explosion_med_distant_01}}
 {id=WavePlayer_3, properties {soundWaveAssetPtr=/Game/Sounds/Fort_Gadgets/Outlander/ShapedCharge_001}}
 {id=WavePlayer_4, properties {soundWaveAssetPtr=/Game/Sounds/Fort_Gadgets/Outlander/ShapedCharge_002}}
 {id=WavePlayer_6, properties {soundWaveAssetPtr=/Game/Sounds/Fort_Gadgets/Outlander/ShapedCharge_004}}
 {id=WavePlayer_7, properties {soundWaveAssetPtr=&/explosion_med_distant_02}}
]
properties {
 bOverrideConcurrency=true
 maxDistance=10579.2871
 soundClassObject=/Game/Sounds/FortSoundClassesAndModes/Fort_Explosives_Large_PSM
 volumeMultiplier=0.5661
 concurrencyOverrides {resolutionRule=PreventNew}
}
