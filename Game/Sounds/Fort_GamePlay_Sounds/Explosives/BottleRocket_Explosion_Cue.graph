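# Sound cue for the bottle rocket explosion, routed to the gadgets and items sound class.
# A mixer layers two random picks of roman candle explosion waves, one stereo-named pair and one mono-named pair, with a quieter sharp explosion.
# Each layer has its own modulator for pitch variation.
# Overrides set natural-sound falloff with distance low-pass filtering, and cap concurrency at two voices, stopping the farthest first.
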SoundCue /Game/Sounds/Fort_GamePlay_Sounds/Explosives/BottleRocket_Explosion_Cue
root=Mixer_0
nodes [
 {id=Mixer_0, children [Modulator_2,Modulator_7,Modulator_8], properties {inputVolume [1,1,0.4]}}
 {id=Modulator_2, children [Random_1], properties {pitchMax=1.2}}
 {id=Modulator_7, children [Random_2], properties {pitchMax=1.2, pitchMin=1.1}}
 {id=Modulator_8, children [WavePlayer_9], properties {pitchMax=1.2, pitchMin=1}}
 {id=Random_1, children [WavePlayer_4,WavePlayer_5], properties {weights [1,1]}}
 {id=Random_2, children [WavePlayer_6,WavePlayer_7], properties {weights [1,1]}}
 {id=WavePlayer_4, properties {soundWaveAssetPtr=&/roman_candle_explo_st_02}}
 {id=WavePlayer_5, properties {soundWaveAssetPtr=&/roman_candle_explo_st_01}}
 {id=WavePlayer_6, properties {soundWaveAssetPtr=&/roman_candle_explo_mono_03}}
 {id=WavePlayer_7, properties {soundWaveAssetPtr=&/roman_candle_explo_mono_02}}
 {id=WavePlayer_9, properties {soundWaveAssetPtr=&/explosion_sharp_01}}
]
properties {
 attenuationSettings=/Game/Sounds/Attenuation/Fort_Explosion_Medium_StW_Att
 bOverrideAttenuation=true
 bOverrideConcurrency=true
 maxDistance=6000
 soundClassObject=&Fort_Gadgets_Items
 volumeMultiplier=1
 attenuationOverrides {distanceAlgorithm=NaturalSound, falloffDistance=3800, lPFRadiusMax=3000, lPFRadiusMin=2000, attenuationShapeExtents {x=2200}}
 concurrencyOverrides {maxCount=2, resolutionRule=StopFarthestThenPreventNew}
}
