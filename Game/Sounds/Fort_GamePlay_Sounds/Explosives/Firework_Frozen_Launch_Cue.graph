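# Sound cue for the frozen firework launch, routed through the third-person ranged weapon fire class with small-rifle attenuation.
# A mixer layers a briefly delayed, pitch-modulated random pick of two roman candle shots with a quieter medium firework launch.
# The launch layer runs through an enveloper whose volume curve holds, then fades to silence over a few seconds.
# A shared per-owner concurrency rule prevents new instances beyond its limit.
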SoundCue /Game/Sounds/Fort_GamePlay_Sounds/Explosives/Firework_Frozen_Launch_Cue
root=Mixer_0
nodes [
 {id=Delay_0, children [Modulator_1], properties {delayMax=0.1, delayMin=0.1}}
 {
  id=Enveloper_0
  children [WavePlayer_0]
  properties {
   pitchCurve {editorCurveData {keys [{time=0}]}}
   volumeCurve {editorCurveData {keys [
      <at 0 1>
      {arriveTangent=-0.6905, interpMode=RCIM_Cubic, leaveTangent=-0.6905, tangentMode=RCTM_User, time=1, value=1}
      {arriveTangent=-0.0019, interpMode=RCIM_Cubic, leaveTangent=-0.0019, tangentMode=RCTM_User, time=2.5919, value=0}
     ]}}
  }
 }
 {id=Mixer_0, children [Delay_0,Enveloper_0], properties {inputVolume [1,0.6]}}
 {id=Modulator_1, children [Random_6], properties {pitchMax=1.2}}
 {id=Random_6, children [WavePlayer_15, WavePlayer_3], properties {weights [1,1]}}
 {id=WavePlayer_0, properties {soundWaveAssetPtr=&/Firework_Medium_Launch_01}}
 {id=WavePlayer_15, properties {soundWaveAssetPtr=&/roman_candle_fire_02}}
 {id=WavePlayer_3, properties {soundWaveAssetPtr=&/roman_candle_fire_01}}
]
properties {
 attenuationSettings=/Game/Sounds/Attenuation/Fort_Weapon_Rifle_Sml_Att
 maxDistance=3000
 soundClassObject=&Fort_WeaponRanged_Fire_3P_PSM
 volumeMultiplier=1
 concurrencyOverrides {bLimitToOwner=true, maxCount=1, resolutionRule=PreventNew}
 concurrencySet [/Game/Sounds/Concurrency_Rules/ThreeMaximum_Owner_PreventNew]
}
