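# Looping fuse sound cue for an explosive, named for a grenade, in the explosives sound class.
# A quality-level switch plays the standard dynamite fuse loop on its first two branches and the optimized loop on the last two.
# Each loop runs through an Enveloper that ramps volume and pitch up over a few seconds, then a Doppler node.
# The cue overrides attenuation with logarithmic falloff and concurrency with a prevent-new rule.
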SoundCue /Game/Sounds/Fort_GamePlay_Sounds/Explosives/Fuse_Grenade_Loop_Cue
root=QualityLevel_0
nodes [
 {id=Doppler_0, children [Enveloper_0], properties {dopplerIntensity=2}}
 {id=Doppler_1, children [Enveloper_1], properties {dopplerIntensity=2}}
 {
  id=Enveloper_0
  children [WavePlayer_10]
  properties {pitchCurve {editorCurveData {keys [{time=0}, <at 3.8132 1.0171>]}}, volumeCurve {editorCurveData {keys [<at 0 0.1>,<at 3.4238 2.0153>]}}}
 }
 {
  id=Enveloper_1
  children [WavePlayer_0]
  properties {pitchCurve {editorCurveData {keys [{time=0}, <at 3.8132 1.0171>]}}, volumeCurve {editorCurveData {keys [<at 0 0.1>,<at 3.4238 2.0153>]}}}
 }
 {id=QualityLevel_0, children [Doppler_0,Doppler_0,Doppler_1,Doppler_1]}
 {id=WavePlayer_0, properties {bLooping=true, soundWaveAssetPtr=&/fuse_dynamite_loop_opt}}
 {id=WavePlayer_10, properties {bLooping=true, soundWaveAssetPtr=&/fuse_dynamite_loop}}
]
properties {
 attenuationSettings=/Game/Sounds/Attenuation/Fort_Explosive_Fuse_Ambient_Att
 bOverrideAttenuation=true
 bOverrideConcurrency=true
 maxDistance=1700
 pitchMultiplier=0.7
 soundClassObject=/Game/Sounds/FortSoundClassesAndModes/Fort_Explosives
 volumeMultiplier=0.4
 attenuationOverrides {distanceAlgorithm=Logarithmic, falloffDistance=1200, attenuationShapeExtents {x=500}}
 concurrencyOverrides {resolutionRule=PreventNew}
}
