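# Sound cue for the burning M80 explosive fuse, routed to the explosives sound class with fuse-specific attenuation.
# Each branch mixes a looping fuse wave with a one-shot stick swing layer, both pitch-randomized, the swing pitched higher.
# The quality node gives the first two levels the standard loop plus two random swings, and the last two an "opt" loop plus one fixed swing.
# A shared concurrency rule limits simultaneous instances.
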SoundCue /Game/Sounds/Fort_GamePlay_Sounds/Explosives/Fuse_M80_Loop_Cue
root=QualityLevel_0
nodes [
 {id=Mixer_0, children [Modulator_0,Modulator_1], properties {inputVolume [0.8,1]}}
 {id=Mixer_2, children [Modulator_5,Modulator_3], properties {inputVolume [0.8,1]}}
 {id=Modulator_0, children [WavePlayer_0], properties {pitchMax=1.1, pitchMin=1}}
 {id=Modulator_1, children [WavePlayer_1], properties {pitchMax=1.4, pitchMin=1.2}}
 {id=Modulator_3, children [Random_2], properties {pitchMax=1.4, pitchMin=1.2}}
 {id=Modulator_5, children [WavePlayer_11], properties {pitchMax=1.1, pitchMin=1}}
 {id=QualityLevel_0, children [Mixer_2,Mixer_2,Mixer_0,Mixer_0]}
 {id=Random_2, children [WavePlayer_4,WavePlayer_5], properties {weights [1,1]}}
 {id=WavePlayer_0, properties {bLooping=true, soundWaveAssetPtr=&/fuse_m80_loop_opt}}
 {id=WavePlayer_1, properties {soundWaveAssetPtr=/Game/Sounds/Fort_Weapons/Stick/Stick_Swing02}}
 {id=WavePlayer_11, properties {bLooping=true, soundWaveAssetPtr=&/fuse_m80_loop}}
 {id=WavePlayer_4, properties {soundWaveAssetPtr=&Stick_Swing02}}
 {id=WavePlayer_5, properties {soundWaveAssetPtr=/Game/Sounds/Fort_Weapons/Stick/Stick_Swing01}}
]
properties {
 attenuationSettings=/Game/Sounds/Attenuation/Fort_Explosive_Fuse_Ambient_Att
 maxDistance=3500
 soundClassObject=/Game/Sounds/FortSoundClassesAndModes/Fort_Explosives
 concurrencyOverrides {resolutionRule=PreventNew}
 concurrencySet [/Game/Sounds/Concurrency_Rules/TwoMaximum_StopOldest_VolumeScale]
}
