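# Sound cue for the whiz-by of a grenade launcher projectile passing through the air, played quietly in the ranged projectile effects sound class.
# An equally weighted Random node picks one of five reused crossbow bullet-whip waves.
# A root Modulator pitches the pick down by a fixed amount, with no pitch variation.
# Overridden attenuation disables reverb send despite an assigned in-air projectile attenuation asset, and the debug flag is on.
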SoundCue /Game/Sounds/Fort_GamePlay_Sounds/Explosives/GrenadeLauncher_Projectile_WhizyBy_Cue
root=Modulator_1
nodes [
 {id=Modulator_1, children [Random_3], properties {pitchMax=0.7, pitchMin=0.7}}
 {id=Random_3, children [WavePlayer_3, WavePlayer_4, WavePlayer_9, WavePlayer_10, WavePlayer_11], properties {weights [1,1,1,1,1]}}
 {id=WavePlayer_10, properties {soundWaveAssetPtr=/Game/Athena/Sounds/Weapons/BulletWhips/Crossbow_Whip_04}}
 {id=WavePlayer_11, properties {soundWaveAssetPtr=/Game/Athena/Sounds/Weapons/BulletWhips/Crossbow_Whip_05}}
 {id=WavePlayer_3, properties {soundWaveAssetPtr=/Game/Athena/Sounds/Weapons/BulletWhips/Crossbow_Whip_01}}
 {id=WavePlayer_4, properties {soundWaveAssetPtr=/Game/Athena/Sounds/Weapons/BulletWhips/Crossbow_Whip_02}}
 {id=WavePlayer_9, properties {soundWaveAssetPtr=/Game/Athena/Sounds/Weapons/BulletWhips/Crossbow_Whip_03}}
]
properties {
 attenuationSettings=/Game/Sounds/Attenuation/Fort_Weapon_Projectile_InAir_Loop_Att
 bDebug=true
 bOverrideAttenuation=true
 maxDistance=1512
 soundClassObject=&Fort_WeaponRanged_Projectile_Effects
 volumeMultiplier=0.1
 attenuationOverrides {
  bEnableReverbSend=false
  bUseComplexCollisionForOcclusion=true
  falloffDistance=1000
  occlusionVolumeAttenuation=0
  stereoSpread=0
  attenuationShapeExtents {x=512}
 }
}
