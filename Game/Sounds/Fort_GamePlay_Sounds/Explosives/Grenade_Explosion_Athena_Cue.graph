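# Sound cue for the Athena grenade explosion, played as a large explosive with its own concurrency limit.
# A quality-level switch picks one of three mixes, each blending a close layer with a quieter medium or distant layer under separate attenuation.
# The first level randomizes two waves per layer, the second uses one each, and the last two share a mix of optimized waves.
# Modulators add pitch variation per layer.
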SoundCue /Game/Sounds/Fort_GamePlay_Sounds/Explosives/Grenade_Explosion_Athena_Cue
root=QualityLevel_0
nodes [
 {
  id=Attenuation_0
  children [Modulator_13]
  properties {attenuationSettings=&Fort_Weapon_Projectile_Medium_Mono_Athena_Att}
 }
 {id=Attenuation_1, children [Modulator_12], properties {attenuationSettings=&Fort_Weapon_Projectile_Close_Athena_Att}}
 {
  id=Attenuation_2
  children [Modulator_15]
  properties {attenuationSettings=&Fort_Weapon_Projectile_Medium_Mono_Athena_Att}
 }
 {id=Attenuation_3, children [Modulator_14], properties {attenuationSettings=&Fort_Weapon_Projectile_Close_Athena_Att}}
 {
  id=Attenuation_4
  children [Modulator_9]
  properties {attenuationSettings=&Fort_Weapon_Projectile_Medium_Mono_Athena_Att}
 }
 {id=Attenuation_5, children [Modulator_8], properties {attenuationSettings=&Fort_Weapon_Projectile_Close_Athena_Att}}
 {id=Mixer_2, children [Attenuation_1, Attenuation_0], properties {inputVolume [0.8,1]}}
 {id=Mixer_5, children [Attenuation_3, Attenuation_2], properties {inputVolume [0.8,1]}}
 {id=Mixer_6, children [Attenuation_5, Attenuation_4], properties {inputVolume [0.8,1]}}
 {id=Modulator_12, children [Random_3], properties {pitchMax=1.1}}
 {id=Modulator_13, children [Random_4], properties {pitchMax=1.1, pitchMin=0.9, volumeMax=0.7, volumeMin=0.7}}
 {id=Modulator_14, children [WavePlayer_21]}
 {id=Modulator_15, children [WavePlayer_22], properties {pitchMax=1.1, pitchMin=0.9, volumeMax=0.7, volumeMin=0.7}}
 {id=Modulator_8, children [WavePlayer_16], properties {pitchMax=1.1, pitchMin=0.8}}
 {id=Modulator_9, children [WavePlayer_17], properties {pitchMax=1.1, pitchMin=0.8, volumeMax=0.7, volumeMin=0.7}}
 {id=QualityLevel_0, children [Mixer_2,Mixer_5,Mixer_6,Mixer_6]}
 {id=Random_3, children [WavePlayer_10, WavePlayer_12], properties {weights [1,1]}}
 {id=Random_4, children [WavePlayer_15, WavePlayer_20], properties {weights [1,1]}}
 {id=WavePlayer_10, properties {soundWaveAssetPtr=&/Grenade_Explode_Athena_Close_01}}
 {id=WavePlayer_12, properties {soundWaveAssetPtr=&/Grenade_Explode_Athena_Close_02}}
 {id=WavePlayer_15, properties {soundWaveAssetPtr=&/Grenade_Explode_Athena_Medium_01}}
 {id=WavePlayer_16, properties {soundWaveAssetPtr=&/Grenade_Explosion_Athena_close_opt}}
 {id=WavePlayer_17, properties {soundWaveAssetPtr=&/Grenade_Explosion_Athena_distant_opt}}
 {id=WavePlayer_20, properties {soundWaveAssetPtr=&/Grenade_Explode_Athena_Medium_02}}
 {id=WavePlayer_21, properties {soundWaveAssetPtr=&Grenade_Explode_Athena_Close_01}}
 {id=WavePlayer_22, properties {soundWaveAssetPtr=&Grenade_Explode_Athena_Medium_01}}
]
properties {
 attenuationSettings=/Game/Sounds/Attenuation/Fort_Explosion_Far_StW_Att
 bOverrideConcurrency=true
 maxDistance=11000
 soundClassObject=/Game/Sounds/FortSoundClassesAndModes/Fort_Explosives_Large_PSM
 concurrencyOverrides {maxCount=3}
}
