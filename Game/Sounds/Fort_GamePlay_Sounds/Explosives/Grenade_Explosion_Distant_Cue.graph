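# Sound cue for a distant grenade explosion, using far explosion attenuation and the medium explosives sound class.
# A quality-level node plays a two-layer mix on its first two levels and a single pitch-modulated "opt" wave on the last two.
# The mix layers a pitch-randomized sharp explosion with a random pick of two sharp variants.
# A concurrency override caps simultaneous instances at a small count.
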SoundCue /Game/Sounds/Fort_GamePlay_Sounds/Explosives/Grenade_Explosion_Distant_Cue
root=QualityLevel_0
nodes [
 {id=Mixer_0, children [Modulator_1,Random_2], properties {inputVolume [1,1]}}
 {id=Modulator_0, children [WavePlayer_0], properties {pitchMax=1.1, pitchMin=0.88}}
 {id=Modulator_1, children [WavePlayer_1], properties {pitchMax=1.6, pitchMin=0.8}}
 {id=QualityLevel_0, children [Mixer_0,Mixer_0,Modulator_0,Modulator_0]}
 {id=Random_2, children [WavePlayer_20, WavePlayer_21], properties {weights [1,1]}}
 {id=WavePlayer_0, properties {soundWaveAssetPtr=&/Grenade_Explosion_Distant_opt}}
 {id=WavePlayer_1, properties {soundWaveAssetPtr=&/fort_distant_explosion_sharp}}
 {id=WavePlayer_20, properties {soundWaveAssetPtr=&/distant_explosion_sharp_02}}
 {id=WavePlayer_21, properties {soundWaveAssetPtr=&/distant_explosion_sharp_01}}
]
properties {
 attenuationSettings=/Game/Sounds/Attenuation/Fort_Explosion_Far_StW_Att
 bOverrideConcurrency=true
 maxDistance=11000
 soundClassObject=/Game/Sounds/FortSoundClassesAndModes/Fort_Explosives_Medium
 volumeMultiplier=1.35
 concurrencyOverrides {maxCount=3}
}
