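# Sound cue for the grenade launcher explosion in Athena, blending a nearer layer with a distant layer.
# The nearer layer mixes a grenade projectile blast and a sharp mono explosion through medium-explosion attenuation, a modulator and a third-person fire sound class.
# The distant layer pitch-varies one distant explosion wave through distant attenuation and a distant fire sound class.
# The cue references a weapon explosion concurrency rule to limit overlapping blasts.
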SoundCue /Game/Sounds/Fort_GamePlay_Sounds/Explosives/Grenade_Launcher_Explosion_MLTI_Athena_Cue
root=Mixer_4
nodes [
 {
  id=Attenuation_5
  children [Modulator_5]
  properties {
   attenuationSettings=/Game/Sounds/Attenuation/Fort_Weapon_Rocket_Distant_Explosion_Athena_Att
   attenuationOverrides {attenuationShapeExtents {x=1000}}
  }
 }
 {
  id=Attenuation_6
  children [Mixer_0]
  properties {attenuationSettings=/Game/Sounds/Attenuation/Fort_Weapon_Rocket_Medium_Explosion_Athena_Att, attenuationOverrides {falloffDistance=1600}}
 }
 {id=Mixer_0, children [WavePlayer_1,WavePlayer_2], properties {inputVolume [0.4,1]}}
 {id=Mixer_4, children [SoundClass_6,SoundClass_2], properties {inputVolume [1,1.2]}}
 {id=Modulator_5, children [WavePlayer_4], properties {pitchMax=1.1, pitchMin=0.9}}
 {id=Modulator_6, children [Attenuation_6]}
 {
  id=SoundClass_2
  children [Attenuation_5]
  properties {soundClassOverride=&Fort_WeaponRanged_Fire_Distant}
 }
 {id=SoundClass_6, children [Modulator_6], properties {soundClassOverride=&Fort_WeaponRanged_Fire_3P_PSM}}
 {id=WavePlayer_1, properties {soundWaveAssetPtr=&/Explosion_GrenadeProjectile_01_E}}
 {id=WavePlayer_2, properties {soundWaveAssetPtr=&/explosion_sharp_mono_01}}
 {id=WavePlayer_4, properties {soundWaveAssetPtr=&/explosion_sharp_distant_01}}
]
properties {
 maxDistance=15000
 volumeMultiplier=0.2
 concurrencyOverrides {maxCount=4, resolutionRule=StopOldest, volumeScale=0.6}
 concurrencySet [/Game/Sounds/Concurrency_Rules/Weapon_Explosion_Conc]
}
