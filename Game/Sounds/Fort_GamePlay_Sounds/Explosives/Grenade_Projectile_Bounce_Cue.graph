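# Sound cue for a grenade projectile bounce, routed to the ranged weapon projectile effects sound class.
# A random node picks one of three equally weighted mono bounce waves, feeding a modulator that pins pitch at its original value.
# It uses the grenade bounce attenuation asset at reduced volume, with a concurrency override that blocks new plays once a few are active.
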SoundCue /Game/Sounds/Fort_GamePlay_Sounds/Explosives/Grenade_Projectile_Bounce_Cue
root=Modulator_5
nodes [
 {id=Modulator_5, children [Random_4], properties {pitchMax=1, pitchMin=1}}
 {id=Random_4, children [WavePlayer_28, WavePlayer_29, WavePlayer_30], properties {weights [1,1,1]}}
 {id=WavePlayer_28, properties {soundWaveAssetPtr=&/Grenade_Bounce_Mono_01}}
 {id=WavePlayer_29, properties {soundWaveAssetPtr=&/Grenade_Bounce_Mono_02}}
 {id=WavePlayer_30, properties {soundWaveAssetPtr=&/Grenade_Bounce_Mono_03}}
]
properties {
 attenuationSettings=/Game/Sounds/Attenuation/Fort_Grenade_Bounce_Att
 bOverrideConcurrency=true
 maxDistance=5500
 soundClassObject=&Fort_WeaponRanged_Projectile_Effects
 volumeMultiplier=0.65
 concurrencyOverrides {maxCount=3, resolutionRule=PreventNew}
}
