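# Sound cue for the distant whip sound of a grenade projectile in flight.
# A random node picks one of three equally weighted frag grenade far-whip waves and feeds it through a modulator.
# It uses a dedicated far-whip attenuation, routes to the ranged weapon projectile effects sound class at reduced volume, and overrides concurrency to cap simultaneous instances.
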SoundCue /Game/Sounds/Fort_GamePlay_Sounds/Explosives/Grenade_Projectile_Whip_Far_Cue
root=Modulator_0
nodes [
 {id=Modulator_0, children [Random_0]}
 {id=Random_0, children [WavePlayer_0,WavePlayer_1,WavePlayer_2], properties {bShouldExcludeFromBranchCulling=true, weights [1,1,1]}}
 {id=WavePlayer_0, properties {soundWaveAssetPtr=&/Grenade_Frag_Whip_Far_01}}
 {id=WavePlayer_1, properties {soundWaveAssetPtr=&/Grenade_Frag_Whip_Far_03}}
 {id=WavePlayer_2, properties {soundWaveAssetPtr=&/Grenade_Frag_Whip_Far_02}}
]
properties {
 attenuationSettings=/Game/Sounds/Attenuation/Fort_Grenade_Prj_Whip_Far_Att
 bOverrideConcurrency=true
 maxDistance=9000
 soundClassObject=&Fort_WeaponRanged_Projectile_Effects
 volumeMultiplier=0.6
 concurrencyOverrides {maxCount=2}
}
