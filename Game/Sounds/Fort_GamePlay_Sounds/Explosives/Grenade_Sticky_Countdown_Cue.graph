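# Sound cue for the sticky grenade countdown sound, classed under gadget and item sounds.
# A single wave player feeds the Grenade_Sticky_Countdown05 wave into a mixer root that boosts its input volume.
# Spatial falloff comes from the shared medium explosive-beep attenuation asset, with no randomization or looping node in the graph.
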SoundCue /Game/Sounds/Fort_GamePlay_Sounds/Explosives/Grenade_Sticky_Countdown_Cue
root=Mixer_0
nodes [
 {id=Mixer_0, children [WavePlayer_1], properties {inputVolume [1.2]}}
 {id=WavePlayer_1, properties {soundWaveAssetPtr=&/Grenade_Sticky_Countdown05}}
]
properties {
 attenuationSettings=/Game/Sounds/Attenuation/Fort_ExplosiveBeep_Medium_Att
 maxDistance=6500
 soundClassObject=&Fort_Gadgets_Items
 volumeMultiplier=1
}
