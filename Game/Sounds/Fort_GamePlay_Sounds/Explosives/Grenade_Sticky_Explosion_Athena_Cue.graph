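# Sound cue for the Athena sticky grenade explosion, routed to the large explosives sound class.
# A root quality switch picks one of four mixers, each layering close and medium-distance mono sounds with their own attenuation and modulation.
# The first tier randomizes three waves per layer, with a nested quality switch on the medium one; later tiers use fewer waves, ending in single optimized close sounds.
# A concurrency override prevents new instances once the limit is reached.
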SoundCue /Game/Sounds/Fort_GamePlay_Sounds/Explosives/Grenade_Sticky_Explosion_Athena_Cue
root=QualityLevel_0
nodes [
 {
  id=Attenuation_0
  children [Modulator_13]
  properties {attenuationSettings=&Fort_Weapon_Projectile_Medium_Mono_Athena_Att}
 }
 {id=Attenuation_1, children [Modulator_12], properties {attenuationSettings=&Fort_Weapon_Projectile_Close_Athena_Att}}
 {
  id=Attenuation_2
  children [Modulator_15]
  properties {attenuationSettings=&Fort_Weapon_Projectile_Medium_Mono_Athena_Att}
 }
 {id=Attenuation_3, children [Modulator_14], properties {attenuationSettings=&Fort_Weapon_Projectile_Close_Athena_Att}}
 {
  id=Attenuation_4
  children [Modulator_9]
  properties {attenuationSettings=&Fort_Weapon_Projectile_Medium_Mono_Athena_Att}
 }
 {id=Attenuation_5, children [Modulator_8], properties {attenuationSettings=&Fort_Weapon_Projectile_Close_Athena_Att}}
 {
  id=Attenuation_6
  children [Modulator_11]
  properties {attenuationSettings=&Fort_Weapon_Projectile_Medium_Mono_Athena_Att}
 }
 {id=Attenuation_7, children [Modulator_10], properties {attenuationSettings=&Fort_Weapon_Projectile_Close_Athena_Att}}
 {id=Mixer_2, children [Attenuation_1, Attenuation_0], properties {inputVolume [1,1]}}
 {id=Mixer_5, children [Attenuation_3, Attenuation_2], properties {inputVolume [1,1]}}
 {id=Mixer_6, children [Attenuation_5, Attenuation_4], properties {inputVolume [1,1]}}
 {id=Mixer_7, children [Attenuation_7, Attenuation_6], properties {inputVolume [1,1]}}
 {id=Modulator_10, children [WavePlayer_2], properties {pitchMax=1.1, pitchMin=0.8}}
 {id=Modulator_11, children [WavePlayer_49], properties {pitchMax=1.1, pitchMin=0.8}}
 {id=Modulator_12, children [Random_5], properties {pitchMax=1, pitchMin=0.9}}
 {id=Modulator_13, children [QualityLevel_1], properties {pitchMax=1.15, pitchMin=0.85}}
 {id=Modulator_14, children [Random_6]}
 {id=Modulator_15, children [WavePlayer_44]}
 {id=Modulator_8, children [WavePlayer_1], properties {pitchMax=1.1, pitchMin=0.8}}
 {id=Modulator_9, children [WavePlayer_47], properties {pitchMax=1.1, pitchMin=0.8}}
 {id=QualityLevel_0, children [Mixer_2,Mixer_5,Mixer_6,Mixer_7]}
 {id=QualityLevel_1, children [Random_3,Random_3,WavePlayer_0,WavePlayer_0]}
 {id=Random_3, children [WavePlayer_10, WavePlayer_11, WavePlayer_12], properties {weights [1,1,1]}}
 {id=Random_5, children [WavePlayer_26, WavePlayer_38, WavePlayer_39], properties {weights [1,1,1]}}
 {id=Random_6, children [WavePlayer_43, WavePlayer_45], properties {weights [1,1]}}
 {id=WavePlayer_0, properties {soundWaveAssetPtr=&/Grenade_Sticky_Explode_medium_01_opt}}
 {id=WavePlayer_1, properties {soundWaveAssetPtr=&/Grenade_Sticky_Explode01_opt}}
 {id=WavePlayer_10, properties {soundWaveAssetPtr=&/Grenade_Sticky_Explode_medium_01}}
 {id=WavePlayer_11, properties {soundWaveAssetPtr=&/Grenade_Sticky_Explode_medium_02}}
 {id=WavePlayer_12, properties {soundWaveAssetPtr=&/Grenade_Sticky_Explode_medium_03}}
 {id=WavePlayer_2, properties {soundWaveAssetPtr=&Grenade_Sticky_Explode01_opt}}
 {id=WavePlayer_26, properties {soundWaveAssetPtr=&/Grenade_Sticky_Explode01}}
 {id=WavePlayer_38, properties {soundWaveAssetPtr=&/Grenade_Sticky_Explode02}}
 {id=WavePlayer_39, properties {soundWaveAssetPtr=&/Grenade_Sticky_Explode03}}
 {id=WavePlayer_43, properties {soundWaveAssetPtr=&Grenade_Sticky_Explode01}}
 {id=WavePlayer_44, properties {soundWaveAssetPtr=&Grenade_Sticky_Explode_medium_01}}
 {id=WavePlayer_45, properties {soundWaveAssetPtr=&Grenade_Sticky_Explode02}}
 {id=WavePlayer_47, properties {soundWaveAssetPtr=&Grenade_Sticky_Explode_medium_01}}
 {id=WavePlayer_49, properties {soundWaveAssetPtr=&Grenade_Sticky_Explode_medium_01}}
]
properties {
 attenuationSettings=/Game/Sounds/Attenuation/Fort_Explosion_Far_StW_Att
 bOverrideConcurrency=true
 maxDistance=11000
 soundClassObject=/Game/Sounds/FortSoundClassesAndModes/Fort_Explosives_Large_PSM
 volumeMultiplier=0.6369
 concurrencyOverrides {resolutionRule=PreventNew}
}
