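# Sound cue for a sticky grenade landing, routed to the ranged weapon projectile effects sound class.
# A mixer layers the landing wave with a slightly quieter wobble wave, with no randomization.
# Inline attenuation overrides the referenced grenade bounce settings, using logarithmic falloff and a distance-based low-pass filter.
# Inline concurrency allows one instance per owner and blocks new plays while one is active.
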SoundCue /Game/Sounds/Fort_GamePlay_Sounds/Explosives/Grenade_Sticky_Land_Cue
root=Mixer_0
nodes [
 {id=Mixer_0, children [WavePlayer_2,WavePlayer_0], properties {inputVolume [1,0.75]}}
 {id=WavePlayer_0, properties {soundWaveAssetPtr=&/Grenade_Sticky_Land_Wobble}}
 {id=WavePlayer_2, properties {soundWaveAssetPtr=&/Grenade_Sticky_Land01}}
]
properties {
 attenuationSettings=/Game/Sounds/Attenuation/Fort_Grenade_Bounce_Att
 bOverrideAttenuation=true
 bOverrideConcurrency=true
 maxDistance=3000
 soundClassObject=&Fort_WeaponRanged_Projectile_Effects
 volumeMultiplier=1
 attenuationOverrides {
  bAttenuateWithLPF=true
  distanceAlgorithm=Logarithmic
  falloffDistance=2500
  lPFFrequencyAtMax=2000
  lPFRadiusMax=3000
  lPFRadiusMin=500
  attenuationShapeExtents {x=500}
 }
 concurrencyOverrides {bLimitToOwner=true, maxCount=1, resolutionRule=PreventNew}
}
