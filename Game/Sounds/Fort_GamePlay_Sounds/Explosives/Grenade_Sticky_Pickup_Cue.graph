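# Sound cue for picking up a sticky grenade, categorized under the gadgets and items sound class.
# A single wave player plays the Grenade_Sticky_Pickup sound with no randomization or mixing.
# It uses the shared weapon pickup attenuation and a reduced volume.
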SoundCue /Game/Sounds/Fort_GamePlay_Sounds/Explosives/Grenade_Sticky_Pickup_Cue
root=WavePlayer_0
nodes [{id=WavePlayer_0, properties {soundWaveAssetPtr=&/Grenade_Sticky_Pickup}}]
properties {
 attenuationSettings=&Fort_Weapon_Pickup_Att
 maxDistance=3500
 soundClassObject=&Fort_Gadgets_Items
 volumeMultiplier=0.25
}
