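# Sound cue for a sticky grenade sticking to a player, routed to the Fort_Enemy_Damaged_PSM sound class.
# A Mixer layers the stick-to-player wave with the landing wobble wave, weighting the wobble louder, under a reduced cue volume.
# An attenuation override turns off distance volume falloff, superseding the referenced grenade-bounce attenuation asset.
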SoundCue /Game/Sounds/Fort_GamePlay_Sounds/Explosives/Grenade_Sticky_StickPlayer_Cue
root=Mixer_0
nodes [
 {id=Mixer_0, children [WavePlayer_1,WavePlayer_0], properties {inputVolume [1,2]}}
 {id=WavePlayer_0, properties {soundWaveAssetPtr=&/Grenade_Sticky_Land_Wobble}}
 {id=WavePlayer_1, properties {soundWaveAssetPtr=&/Grenade_Sticky_StickPlayer_01}}
]
properties {
 attenuationSettings=/Game/Sounds/Attenuation/Fort_Grenade_Bounce_Att
 bOverrideAttenuation=true
 maxDistance=2097152
 soundClassObject=/Game/Sounds/FortSoundClassesAndModes/Fort_Enemy_Damaged_PSM
 volumeMultiplier=0.45
 attenuationOverrides {bAttenuate=false, stereoSpread=1000}
}
