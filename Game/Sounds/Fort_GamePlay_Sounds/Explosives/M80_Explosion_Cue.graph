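# Sound cue for the M80 explosive's explosion, built from repurposed Husk monster sound effects.
# A mixer layers a random lobber projectile impact, pitched up to double speed, with a random monster whoosh at varied pitch.
# An outer modulator raises and randomizes the pitch of the combined result.
# It plays quietly in the weapons sound class with medium explosion attenuation and a concurrency override capping simultaneous instances.
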SoundCue /Game/Sounds/Fort_GamePlay_Sounds/Explosives/M80_Explosion_Cue
root=Modulator_0
nodes [
 {id=Mixer_1, children [Modulator_1,Modulator_2], properties {inputVolume [1,1]}}
 {id=Modulator_0, children [Mixer_1], properties {pitchMax=1.4, pitchMin=1.2}}
 {id=Modulator_1, children [Random_1], properties {pitchMax=2, pitchMin=2}}
 {id=Modulator_2, children [Random_2], properties {pitchMax=1.4, pitchMin=0.9}}
 {id=Random_1, children [WavePlayer_4,WavePlayer_5,WavePlayer_6], properties {weights [1,1,1]}}
 {id=Random_2, children [WavePlayer_9,WavePlayer_7], properties {weights [1,1]}}
 {id=WavePlayer_4, properties {soundWaveAssetPtr=/Game/Sounds/Fort_Monster_Husk/SFX/lobber_proj_impact_01}}
 {id=WavePlayer_5, properties {soundWaveAssetPtr=/Game/Sounds/Fort_Monster_Husk/SFX/lobber_proj_impact_02}}
 {id=WavePlayer_6, properties {soundWaveAssetPtr=/Game/Sounds/Fort_Monster_Husk/SFX/lobber_proj_impact_03}}
 {id=WavePlayer_7, properties {soundWaveAssetPtr=/Game/Sounds/Fort_Monster_Husk/SFX/zax_deenen_whooshes_moremonsters_01_6}}
 {id=WavePlayer_9, properties {soundWaveAssetPtr=/Game/Sounds/Fort_Monster_Husk/SFX/zax_deenen_whooshes_moremonsters_01_16}}
]
properties {
 attenuationSettings=/Game/Sounds/Attenuation/Fort_Explosion_Medium_StW_Att
 bOverrideConcurrency=true
 maxDistance=7000
 soundClassObject=/Game/Sounds/FortSoundClassesAndModes/Fort_Weapons
 volumeMultiplier=0.3704
 concurrencyOverrides {maxCount=3, volumeScale=0.8}
}
