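# Sound cue for the propane tank explosion, built from reused grenade projectile explosion waves.
# A Random node picks one of two equally weighted waves, and a Modulator at default settings feeds the output.
# The cue borrows gadget-activate attenuation, uses the medium explosives sound class at slightly reduced volume, and references a stop-oldest tank explosion concurrency asset.
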SoundCue /Game/Sounds/Fort_GamePlay_Sounds/Explosives/PropaneTank_Explosion_Cue
root=Modulator_0
nodes [
 {id=Modulator_0, children [Random_2]}
 {id=Random_2, children [WavePlayer_0,WavePlayer_2], properties {weights [1,1]}}
 {id=WavePlayer_0, properties {soundWaveAssetPtr=&/Explosion_GrenadeProjectile_01_C}}
 {id=WavePlayer_2, properties {soundWaveAssetPtr=&/Explosion_GrenadeProjectile_01_E}}
]
properties {
 attenuationSettings=&Fort_GadgetActivate_Att
 maxDistance=4000
 soundClassObject=/Game/Sounds/FortSoundClassesAndModes/Fort_Explosives_Medium_PSM
 volumeMultiplier=0.8
 concurrencyOverrides {resolutionRule=PreventNew}
 concurrencySet [/Game/Sounds/Concurrency_Rules/TankExplosion_StopOldest_Conc]
}
