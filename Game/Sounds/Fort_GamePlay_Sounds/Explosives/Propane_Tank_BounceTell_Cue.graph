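# Sound cue for a propane tank bounce tell, classed as weapon foley and spatialized with bomb bounce attenuation.
# A mixer layers a quiet, pitched-up random big metal demolition impact under a full-level random propane cylinder step sound.
# Each layer runs through its own modulator for variation.
# A concurrency override allows only a few simultaneous instances and prevents new ones beyond that limit.
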SoundCue /Game/Sounds/Fort_GamePlay_Sounds/Explosives/Propane_Tank_BounceTell_Cue
root=Mixer_0
nodes [
 {id=Mixer_0, children [Modulator_0,Modulator_1], properties {inputVolume [0.2,1]}}
 {id=Modulator_0, children [Random_0], properties {pitchMax=1.4, pitchMin=1.2}}
 {id=Modulator_1, children [Random_2]}
 {id=Random_0, children [WavePlayer_0,WavePlayer_1,WavePlayer_2,WavePlayer_3], properties {weights [1,1,1,1]}}
 {id=Random_2, children [WavePlayer_4,WavePlayer_5,WavePlayer_6], properties {weights [1,1,1]}}
 {id=WavePlayer_0, properties {soundWaveAssetPtr=/Game/Sounds/Fort_Impact_Sounds/Demolition/Metal/Metal_Big_01}}
 {id=WavePlayer_1, properties {soundWaveAssetPtr=/Game/Sounds/Fort_Impact_Sounds/Demolition/Metal/Metal_Big_02}}
 {id=WavePlayer_2, properties {soundWaveAssetPtr=/Game/Sounds/Fort_Impact_Sounds/Demolition/Metal/Metal_Big_04}}
 {id=WavePlayer_3, properties {soundWaveAssetPtr=/Game/Sounds/Fort_Impact_Sounds/Demolition/Metal/Metal_Big_06}}
 {id=WavePlayer_4, properties {soundWaveAssetPtr=&/propane_cylinder_step_04}}
 {id=WavePlayer_5, properties {soundWaveAssetPtr=&/propane_cylinder_step_01}}
 {id=WavePlayer_6, properties {soundWaveAssetPtr=&/propane_cylinder_step_03}}
]
properties {
 attenuationSettings=/Game/Sounds/Attenuation/Fort_Bomb_Bounce_Att
 bOverrideConcurrency=true
 maxDistance=2000
 soundClassObject=&Fort_WeaponFoley
 volumeMultiplier=1.002
 concurrencyOverrides {maxCount=3, resolutionRule=PreventNew}
}
