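# Sound cue for a proximity mine explosion, in the gadgets/items sound class with medium explosion attenuation.
# After a brief fixed delay, a quality-level switch picks one of four pitch-modulated branches.
# The first two layer a mono blast with delayed stereo blast and later debris, the first randomizing variants; the last two play one optimized wave.
# A concurrency override prevents new instances at the limit.
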SoundCue /Game/Sounds/Fort_GamePlay_Sounds/Explosives/ProximityMine_Explosion_Cue
root=Delay_5
nodes [
 {id=Delay_0, children [Modulator_2], properties {delayMax=0.25, delayMin=0.2}}
 {id=Delay_1, children [Random_1], properties {delayMax=0.8, delayMin=0.6}}
 {id=Delay_2, children [Modulator_6], properties {delayMax=0.25, delayMin=0.2}}
 {id=Delay_3, children [WavePlayer_10], properties {delayMax=0.8, delayMin=0.6}}
 {id=Delay_5, children [QualityLevel_0], properties {delayMax=0.05, delayMin=0.05}}
 {id=Mixer_0, children [Modulator_5,Delay_2,Modulator_7], properties {inputVolume [1,0.4,0.7]}}
 {id=Mixer_1, children [Modulator_1,Delay_0,Modulator_3], properties {inputVolume [1,0.4,0.7]}}
 {id=Modulator_0, children [WavePlayer_1], properties {pitchMax=1.1, pitchMin=0.9}}
 {id=Modulator_1, children [WavePlayer_2], properties {pitchMax=1.2, pitchMin=1}}
 {id=Modulator_2, children [Random_0], properties {pitchMax=1.6, pitchMin=1}}
 {id=Modulator_3, children [Delay_1], properties {pitchMax=1.1, pitchMin=0.8}}
 {id=Modulator_4, children [WavePlayer_7], properties {pitchMax=1.1, pitchMin=0.9}}
 {id=Modulator_5, children [WavePlayer_0], properties {pitchMax=1.2, pitchMin=1}}
 {id=Modulator_6, children [WavePlayer_8], properties {pitchMax=1.6, pitchMin=1}}
 {id=Modulator_7, children [Delay_3], properties {pitchMax=1.1, pitchMin=0.8}}
 {id=QualityLevel_0, children [Mixer_1,Mixer_0,Modulator_0,Modulator_4]}
 {id=Random_0, children [WavePlayer_3,WavePlayer_4], properties {weights [1,1]}}
 {id=Random_1, children [WavePlayer_5,WavePlayer_6], properties {weights [1,1]}}
 {id=WavePlayer_0, properties {soundWaveAssetPtr=&/bomb_explode_mono_01}}
 {id=WavePlayer_1, properties {soundWaveAssetPtr=&/bomb_explode_opt}}
 {id=WavePlayer_10, properties {soundWaveAssetPtr=&/bomb_debris_01}}
 {id=WavePlayer_2, properties {soundWaveAssetPtr=&bomb_explode_mono_01}}
 {id=WavePlayer_3, properties {soundWaveAssetPtr=&/bomb_explode_stereo_01}}
 {id=WavePlayer_4, properties {soundWaveAssetPtr=&/bomb_explode_stereo_02}}
 {id=WavePlayer_5, properties {soundWaveAssetPtr=&bomb_debris_01}}
 {id=WavePlayer_6, properties {soundWaveAssetPtr=&/bomb_debris_02}}
 {id=WavePlayer_7, properties {soundWaveAssetPtr=&bomb_explode_opt}}
 {id=WavePlayer_8, properties {soundWaveAssetPtr=&bomb_explode_stereo_01}}
]
properties {
 attenuationSettings=/Game/Sounds/Attenuation/Fort_Explosion_Medium_StW_Att
 bOverrideConcurrency=true
 maxDistance=7000
 soundClassObject=&Fort_Gadgets_Items
 volumeMultiplier=0.9
 concurrencyOverrides {resolutionRule=PreventNew}
}
