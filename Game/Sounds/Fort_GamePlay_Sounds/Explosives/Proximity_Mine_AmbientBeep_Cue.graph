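# Sound cue for the looping ambient beep of the proximity mine, classed under gadget and item sounds.
# A Delay node with a short random range feeds a looping wave player, so each instance begins its beep loop at a slightly different offset.
# Small explosive-beep attenuation spatializes it, and a concurrency override caps simultaneous copies, refusing new ones once the limit is reached.
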SoundCue /Game/Sounds/Fort_GamePlay_Sounds/Explosives/Proximity_Mine_AmbientBeep_Cue
root=Delay_0
nodes [
 {id=Delay_0, children [WavePlayer_0], properties {delayMax=1, delayMin=0.5}}
 {id=WavePlayer_0, properties {bLooping=true, soundWaveAssetPtr=&/timer_beep_lo3}}
]
properties {
 attenuationSettings=/Game/Sounds/Attenuation/Fort_ExplosiveBeep_Small_Att
 bOverrideConcurrency=true
 maxDistance=1200
 soundClassObject=&Fort_Gadgets_Items
 volumeMultiplier=1.4
 concurrencyOverrides {maxCount=4, resolutionRule=PreventNew}
}
