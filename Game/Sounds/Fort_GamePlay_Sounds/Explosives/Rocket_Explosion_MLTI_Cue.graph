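# Sound cue for a rocket explosion, layering a near blast and a distant report after a short random delay.
# A mixer blends the two pitch-modulated waves, each with its own attenuation and ranged-weapon sound class override.
# Only the distant layer runs through a distance crossfade, fading in at mid range and out far away; its second input is empty.
# Overridden concurrency blocks new instances rather than stopping playing ones.
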SoundCue /Game/Sounds/Fort_GamePlay_Sounds/Explosives/Rocket_Explosion_MLTI_Cue
root=Delay_1
nodes [
 {
  id=Attenuation_4
  children [WavePlayer_0]
  properties {attenuationSettings=/Game/Sounds/Attenuation/Fort_Explosion_RPG_Att, attenuationOverrides {falloffDistance=1600}}
 }
 {
  id=Attenuation_5
  children [Modulator_5]
  properties {
   attenuationSettings=/Game/Sounds/Attenuation/Fort_Weapon_Projectile_Distant_StW_Att
   attenuationOverrides {falloffDistance=3000, attenuationShapeExtents {x=1000}}
  }
 }
 {id=Delay_1, children [Mixer_4], properties {delayMax=0.1, delayMin=0.05}}
 {
  id=DistanceCrossFade_1
  children [Attenuation_5, null]
  properties {crossFadeInput [
    {fadeInDistanceEnd=3000, fadeInDistanceStart=1500, fadeOutDistanceEnd=12000, fadeOutDistanceStart=10000, volume=1}
    {fadeInDistanceEnd=0, fadeInDistanceStart=0, fadeOutDistanceEnd=0, fadeOutDistanceStart=0, volume=1}
   ]}
 }
 {id=Mixer_4, children [SoundClass_3,SoundClass_2], properties {inputVolume [0.85,0.7]}}
 {id=Modulator_0, children [Attenuation_4]}
 {id=Modulator_5, children [WavePlayer_4], properties {pitchMax=1.1, pitchMin=0.9}}
 {
  id=SoundClass_2
  children [DistanceCrossFade_1]
  properties {soundClassOverride=&Fort_WeaponRanged_Fire_Distant}
 }
 {id=SoundClass_3, children [Modulator_0], properties {soundClassOverride=&Fort_WeaponRanged_Fire_3P_PSM}}
 {id=WavePlayer_0, properties {soundWaveAssetPtr=&/Explosion_GrenadeProjectile_01_E}}
 {id=WavePlayer_4, properties {soundWaveAssetPtr=&/explosion_sharp_distant_01}}
]
properties {bOverrideConcurrency=true, maxDistance=20000, volumeMultiplier=1.2, concurrencyOverrides {resolutionRule=PreventNew}}
