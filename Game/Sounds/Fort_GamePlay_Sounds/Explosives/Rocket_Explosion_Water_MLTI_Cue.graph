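# Sound cue for a rocket explosion, named as the water variant, that layers close, medium and far explosion sounds after a short random delay.
# Close and medium layers share an explosives sound class, while the far layer uses a distant weapon-fire class and a distance crossfade.
# Each layer has its own attenuation and pitch modulator.
# The close layer is pitched up, mixing a medium blast with a huge one that uses optimized audio at some quality levels.
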SoundCue /Game/Sounds/Fort_GamePlay_Sounds/Explosives/Rocket_Explosion_Water_MLTI_Cue
root=Delay_1
nodes [
 {
  id=Attenuation_3
  children [Modulator_3]
  properties {
   attenuationSettings=/Game/Sounds/Attenuation/Fort_Explosion_Close_StW_Att
   attenuationOverrides {falloffDistance=900, attenuationShapeExtents {x=100}}
  }
 }
 {
  id=Attenuation_4
  children [Modulator_4]
  properties {attenuationSettings=/Game/Sounds/Attenuation/Fort_Explosion_Medium_StW_Att, attenuationOverrides {falloffDistance=1600}}
 }
 {
  id=Attenuation_5
  children [Modulator_5]
  properties {
   attenuationSettings=/Game/Sounds/Attenuation/Fort_Explosion_Far_StW_Att
   attenuationOverrides {falloffDistance=3000, attenuationShapeExtents {x=1000}}
  }
 }
 {id=Delay_1, children [Mixer_4], properties {delayMax=0.1, delayMin=0.05}}
 {
  id=DistanceCrossFade_1
  children [Attenuation_5, null]
  properties {crossFadeInput [
    {fadeInDistanceEnd=3000, fadeInDistanceStart=1500, fadeOutDistanceEnd=10000, fadeOutDistanceStart=6000, volume=1}
    {fadeInDistanceEnd=0, fadeInDistanceStart=0, fadeOutDistanceEnd=0, fadeOutDistanceStart=0, volume=1}
   ]}
 }
 {id=Mixer_3, children [Attenuation_3, Attenuation_4], properties {inputVolume [1,1]}}
 {id=Mixer_4, children [SoundClass_3,SoundClass_2], properties {inputVolume [1,1]}}
 {id=Mixer_5, children [QualityLevel_0, WavePlayer_6], properties {inputVolume [1,1]}}
 {id=Modulator_3, children [Mixer_5], properties {pitchMax=2, pitchMin=2}}
 {id=Modulator_4, children [WavePlayer_5], properties {pitchMax=1.1, pitchMin=0.9}}
 {id=Modulator_5, children [WavePlayer_4], properties {pitchMax=1.1, pitchMin=0.9}}
 {id=QualityLevel_0, children [WavePlayer_7,WavePlayer_7,WavePlayer_0,WavePlayer_0]}
 {
  id=SoundClass_2
  children [DistanceCrossFade_1]
  properties {soundClassOverride=&Fort_WeaponRanged_Fire_Distant}
 }
 {id=SoundClass_3, children [Mixer_3], properties {soundClassOverride=/Game/Sounds/FortSoundClassesAndModes/Fort_Explosives_Medium_PSM}}
 {id=WavePlayer_0, properties {soundWaveAssetPtr=&/explosion_c4_huge_05_opt}}
 {id=WavePlayer_4, properties {soundWaveAssetPtr=&/explosion_sharp_distant_01}}
 {id=WavePlayer_5, properties {soundWaveAssetPtr=&/explosion_sharp_mono_01}}
 {id=WavePlayer_6, properties {soundWaveAssetPtr=&/explosion_c4_med_01}}
 {id=WavePlayer_7, properties {soundWaveAssetPtr=&/explosion_c4_huge_05}}
]
properties {bOverrideConcurrency=true, maxDistance=11000, volumeMultiplier=1, concurrencyOverrides {resolutionRule=PreventNew}}
