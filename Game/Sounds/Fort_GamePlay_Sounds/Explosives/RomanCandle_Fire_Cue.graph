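# Sound cue for roman candle firing, assigned to the large explosives sound class.
# It randomly picks one of two equally weighted fire waves, then passes it through a modulator that varies pitch for each shot.
# The cue overrides attenuation with its own falloff and shape extents, despite also referencing a medium explosive-beep attenuation asset.
# A concurrency override prevents new instances from starting once the limit is reached.
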SoundCue /Game/Sounds/Fort_GamePlay_Sounds/Explosives/RomanCandle_Fire_Cue
root=Modulator_7
nodes [
 {id=Modulator_7, children [Random_6], properties {pitchMax=1.2}}
 {id=Random_6, children [WavePlayer_15, WavePlayer_16], properties {weights [1,1]}}
 {id=WavePlayer_15, properties {soundWaveAssetPtr=&/roman_candle_fire_02}}
 {id=WavePlayer_16, properties {soundWaveAssetPtr=&/roman_candle_fire_01}}
]
properties {
 attenuationSettings=/Game/Sounds/Attenuation/Fort_ExplosiveBeep_Medium_Att
 bOverrideAttenuation=true
 bOverrideConcurrency=true
 maxDistance=4800
 soundClassObject=/Game/Sounds/FortSoundClassesAndModes/Fort_Explosives_Large_PSM
 volumeMultiplier=1.0614
 attenuationOverrides {falloffDistance=3000, attenuationShapeExtents {x=1800}}
 concurrencyOverrides {resolutionRule=PreventNew}
}
