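# Sound cue for the Athena Snowball Launcher explosion, mixing three layers after a short random delay.
# The close layer blends a quiet grenade explosion with one of two random snowball impacts, alongside a sharp distant explosion and a pitch-varied giggle.
# Each layer has its own sound class and attenuation, with the close and giggle layers sharing medium-range settings.
# Concurrency is overridden to cap simultaneous instances, stopping the oldest.
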SoundCue /Game/Sounds/Fort_GamePlay_Sounds/Explosives/SnowBall_Launcher_Explosion_MLTI_Athena_Cue
root=Delay_1
nodes [
 {
  id=Attenuation_0
  children [WavePlayer_6]
  properties {attenuationSettings=/Game/Sounds/Attenuation/Fort_Weapon_Rocket_Medium_Explosion_Athena_Att, attenuationOverrides {falloffDistance=1600}}
 }
 {
  id=Attenuation_5
  children [Modulator_5]
  properties {
   attenuationSettings=/Game/Sounds/Attenuation/Fort_Weapon_Rocket_Distant_Explosion_Athena_Att
   attenuationOverrides {falloffDistance=3000, attenuationShapeExtents {x=1000}}
  }
 }
 {
  id=Attenuation_6
  children [Mixer_0]
  properties {attenuationSettings=&Fort_Weapon_Rocket_Medium_Explosion_Athena_Att, attenuationOverrides {falloffDistance=1600}}
 }
 {id=Delay_1, children [Mixer_4], properties {delayMax=0.1, delayMin=0.05}}
 {id=Mixer_0, children [WavePlayer_1,Random_0], properties {inputVolume [0.2,1]}}
 {id=Mixer_4, children [SoundClass_6,SoundClass_2,SoundClass_0], properties {inputVolume [1.2,1.2,0.75]}}
 {id=Modulator_0, children [Attenuation_0], properties {pitchMax=1.15, pitchMin=1}}
 {id=Modulator_5, children [WavePlayer_4], properties {pitchMax=1.1, pitchMin=0.9}}
 {id=Modulator_6, children [Attenuation_6], properties {volumeMin=0.75}}
 {id=Random_0, children [WavePlayer_0,WavePlayer_3], properties {weights [1,1]}}
 {id=SoundClass_0, children [Modulator_0], properties {soundClassOverride=&Fort_WeaponRanged_Fire_3P_PSM}}
 {
  id=SoundClass_2
  children [Attenuation_5]
  properties {soundClassOverride=&Fort_WeaponRanged_Fire_Distant}
 }
 {id=SoundClass_6, children [Modulator_6], properties {soundClassOverride=&Fort_WeaponRanged_Fire_3P_PSM}}
 {id=WavePlayer_0, properties {soundWaveAssetPtr=/Game/Sounds/Fort_Weapons/Launchers/EXPLO_SnowBall_Impact_18}}
 {id=WavePlayer_1, properties {soundWaveAssetPtr=&/Explosion_GrenadeProjectile_01_E}}
 {id=WavePlayer_3, properties {soundWaveAssetPtr=/Game/Sounds/Fort_Weapons/Launchers/EXPLO_SnowBall_Impact_19}}
 {id=WavePlayer_4, properties {soundWaveAssetPtr=&/explosion_sharp_distant_01}}
 {id=WavePlayer_6, properties {soundWaveAssetPtr=/Game/Sounds/Fort_Weapons/Launchers/EXPLO_SnowBall_VO_Giggle_19}}
]
properties {
 bOverrideConcurrency=true
 maxDistance=15000
 volumeMultiplier=0.5
 concurrencyOverrides {maxCount=4, resolutionRule=StopOldest, volumeScale=0.5}
}
