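# Looping Sound Cue for a snowball explosive in the water, playing the grenade fuse underwater bubbles loop.
# A Modulator at default settings wraps the single looping Wave Player for slight pitch and volume variation.
# It routes to the Fort_Explosives sound class and uses inline logarithmic attenuation, which overrides the referenced fuse ambient attenuation asset.
# Inline concurrency prevents new instances once the limit is reached.
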SoundCue /Game/Sounds/Fort_GamePlay_Sounds/Explosives/Snowball_InTheWater_Loop_Cue
root=Modulator_0
nodes [
 {id=Modulator_0, children [WavePlayer_1]}
 {
  id=WavePlayer_1
  properties {bLooping=true, soundWaveAssetPtr=&/Grenade_Proj_Fuse_Water_Bubbles_Loop}
 }
]
properties {
 attenuationSettings=/Game/Sounds/Attenuation/Fort_Explosive_Fuse_Ambient_Att
 bOverrideAttenuation=true
 bOverrideConcurrency=true
 maxDistance=1700
 soundClassObject=/Game/Sounds/FortSoundClassesAndModes/Fort_Explosives
 volumeMultiplier=1
 attenuationOverrides {distanceAlgorithm=Logarithmic, falloffDistance=1200, attenuationShapeExtents {x=500}}
 concurrencyOverrides {resolutionRule=PreventNew}
}
