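# Sound cue for a Snowball Launcher projectile in the air, mixing an in-flight loop layer with a "Weee" voice layer.
# The flight layer picks one of two loop waves, varies pitch, swells via a volume envelope and adds exaggerated Doppler.
# After a short random delay, one of three voice clips fades in and out under its own pickaxe attenuation.
# The mix uses RPG explosion attenuation and a concurrency override limiting simultaneous instances.
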SoundCue /Game/Sounds/Fort_GamePlay_Sounds/Explosives/Snowball_IntheAir_Loop
root=Attenuation_1
nodes [
 {id=Attenuation_1, children [Mixer_1], properties {attenuationSettings=/Game/Sounds/Attenuation/Fort_Explosion_RPG_Att}}
 {id=Attenuation_6, children [Enveloper_6], properties {attenuationSettings=/Game/Sounds/Attenuation/Fort_Weapon_Pickaxe_Medium_Att}}
 {id=Delay_0, children [Attenuation_6], properties {delayMax=0.3, delayMin=0.2}}
 {id=Doppler_1, children [Enveloper_1], properties {dopplerIntensity=4}}
 {
  id=Enveloper_1
  children [Modulator_2]
  properties {
   pitchCurve {editorCurveData {keys [{time=0}]}}
   volumeCurve {editorCurveData {keys [<at 0.0002 0.5183>,<at 0.9977 0.6079>,<at 1.905 0.9819>]}}
  }
 }
 {
  id=Enveloper_6
  children [Modulator_1]
  properties {
   pitchCurve {editorCurveData {keys [{time=0}]}}
   volumeCurve {editorCurveData {keys [
      {interpMode=RCIM_Cubic, tangentMode=RCTM_User, time=0, value=0}
      {interpMode=RCIM_Cubic, tangentMode=RCTM_User, time=0.25, value=1}
      <at 5 0>
     ]}}
  }
 }
 {id=Mixer_1, children [Doppler_1,Delay_0], properties {inputVolume [1,1]}}
 {id=Modulator_1, children [Random_0], properties {pitchMax=1, pitchMin=1}}
 {id=Modulator_2, children [Random_2], properties {pitchMax=1.2, pitchMin=1}}
 {id=Random_0, children [WavePlayer_4,WavePlayer_5,WavePlayer_2], properties {weights [1,1,1]}}
 {id=Random_2, children [WavePlayer_0,WavePlayer_1], properties {weights [1,1]}}
 {id=WavePlayer_0, properties {soundWaveAssetPtr=&/WPN_SnowBall_Launcher_IntheAir_LOOP_01}}
 {id=WavePlayer_1, properties {soundWaveAssetPtr=&/WPN_SnowBall_Launcher_IntheAir_LOOP_02}}
 {id=WavePlayer_2, properties {soundWaveAssetPtr=/Game/Sounds/Fort_Weapons/Launchers/WPN_SnowBall_Launcher_VO_Weee_01}}
 {id=WavePlayer_4, properties {soundWaveAssetPtr=/Game/Sounds/Fort_Weapons/Launchers/WPN_SnowBall_Launcher_VO_Weee_12}}
 {id=WavePlayer_5, properties {soundWaveAssetPtr=/Game/Sounds/Fort_Weapons/Launchers/WPN_SnowBall_Launcher_VO_Weee_13}}
]
properties {
 attenuationSettings=&Fort_Explosion_RPG_Att
 bOverrideConcurrency=true
 maxDistance=13000
 soundClassObject=/Game/Sounds/FortSoundClassesAndModes/Fort_Explosives_Large_PSM
 volumeMultiplier=0.4246
 concurrencyOverrides {maxCount=3}
}
