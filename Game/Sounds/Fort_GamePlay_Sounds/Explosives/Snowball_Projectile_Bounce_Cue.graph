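# Sound cue for a snowball projectile bounce, mixing a snow impact with a delayed "ouch" voice reaction.
# One branch randomly picks from four impact waves with lowered pitch.
# The other adds a short random delay, raised pitch and reduced volume to six voice waves, with two empty random slots so some bounces skip the voice.
# It uses grenade bounce attenuation and limits concurrent instances, preventing new ones at the cap.
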SoundCue /Game/Sounds/Fort_GamePlay_Sounds/Explosives/Snowball_Projectile_Bounce_Cue
root=Mixer_0
nodes [
 {id=Delay_0, children [Modulator_0], properties {delayMax=0.1, delayMin=0.05}}
 {id=Mixer_0, children [Modulator_5,Delay_0], properties {inputVolume [1.25,0.95]}}
 {id=Modulator_0, children [Random_1], properties {pitchMax=1.25, pitchMin=1, volumeMax=0.75, volumeMin=0.4}}
 {id=Modulator_5, children [Random_0], properties {pitchMax=1, pitchMin=0.75}}
 {id=Random_0, children [WavePlayer_0,WavePlayer_1,WavePlayer_2,WavePlayer_3], properties {weights [1,1,1,1]}}
 {
  id=Random_1
  children [WavePlayer_4,WavePlayer_5,WavePlayer_6,WavePlayer_7,WavePlayer_8,WavePlayer_9,null,null]
  properties {weights [1,1,1,1,1,1,1,1]}
 }
 {id=WavePlayer_0, properties {soundWaveAssetPtr=/Game/Sounds/Fort_Weapons/Impacts/IMPCT_SnowBall_Hit_4}}
 {id=WavePlayer_1, properties {soundWaveAssetPtr=/Game/Sounds/Fort_Weapons/Impacts/IMPCT_SnowBall_Hit_1}}
 {id=WavePlayer_2, properties {soundWaveAssetPtr=/Game/Sounds/Fort_Weapons/Impacts/IMPCT_SnowBall_Hit_2}}
 {id=WavePlayer_3, properties {soundWaveAssetPtr=/Game/Sounds/Fort_Weapons/Impacts/IMPCT_SnowBall_Hit_3}}
 {id=WavePlayer_4, properties {soundWaveAssetPtr=/Game/Sounds/Fort_Weapons/Launchers/SnowBall_VO_Bounce_Ouch_01}}
 {id=WavePlayer_5, properties {soundWaveAssetPtr=/Game/Sounds/Fort_Weapons/Launchers/SnowBall_VO_Bounce_Ouch_02}}
 {id=WavePlayer_6, properties {soundWaveAssetPtr=/Game/Sounds/Fort_Weapons/Launchers/SnowBall_VO_Bounce_Ouch_03}}
 {id=WavePlayer_7, properties {soundWaveAssetPtr=/Game/Sounds/Fort_Weapons/Launchers/SnowBall_VO_Bounce_Ouch_04}}
 {id=WavePlayer_8, properties {soundWaveAssetPtr=/Game/Sounds/Fort_Weapons/Launchers/SnowBall_VO_Bounce_Ouch_05}}
 {id=WavePlayer_9, properties {soundWaveAssetPtr=/Game/Sounds/Fort_Weapons/Launchers/SnowBall_VO_Bounce_Ouch_06}}
]
properties {
 attenuationSettings=/Game/Sounds/Attenuation/Fort_Grenade_Bounce_Att
 bOverrideConcurrency=true
 maxDistance=5500
 soundClassObject=&Fort_WeaponRanged_Projectile_Effects
 volumeMultiplier=1
 concurrencyOverrides {maxCount=3, resolutionRule=PreventNew}
}
