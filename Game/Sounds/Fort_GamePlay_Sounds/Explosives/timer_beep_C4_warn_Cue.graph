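# Sound cue for the C4 explosive timer's warning beep, classed as a gadget item sound.
# A single wave player feeds the timer_beep_C4_warn wave straight to the output, with no randomization or modulation.
# The shared gadget activation attenuation spatializes it out to a set maximum distance, and the cue plays at reduced volume.
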SoundCue /Game/Sounds/Fort_GamePlay_Sounds/Explosives/timer_beep_C4_warn_Cue
root=WavePlayer_0
nodes [{id=WavePlayer_0, properties {soundWaveAssetPtr=&/timer_beep_C4_warn}}]
properties {
 attenuationSettings=&Fort_GadgetActivate_Att
 maxDistance=4000
 soundClassObject=&Fort_Gadgets_Items
 volumeMultiplier=0.6
}
